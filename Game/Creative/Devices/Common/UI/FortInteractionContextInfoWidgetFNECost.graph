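# Interaction prompt widget for Creative devices that stacks a basic interaction description above an item cost display.
# On each context update, it caches its parent AthenaInteractionIndicator, then sets the context text, color and interaction count, appending the creative name when required.
# With a valid minigame, it feeds the cost widget the receiving actor's FNEDeviceCostComponent, passing no actor outside play mode.
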WidgetBlueprint /Game/Creative/Devices/Common/UI/FortInteractionContextInfoWidgetFNECost
graphs [
 EventGraph {nodes [
   n1 = Event OnContextInfoUpdated!() -> (OutputDelegate, then, ContextInfo:Fort Interact Context Info Object Reference) {is=title K2Node_Event}
   > n2 = Get!(execute:Exec) -> (then:Exec, else:Exec, Indicator Widget:Athena Interaction Indicator Object Reference)
   Update Details From Context!(execute=[@n2.then, @n6.then], self, New Context:Fort Interact Context Info Object Reference=@n1.ContextInfo) -> (then)
   n3 = Get ReceivingActor!(self:Fort Interact Context Info Object Reference=@n1.ContextInfo) -> (ReceivingActor:Actor Weak object)
   n4 = GetMinigame!(ContextActor:Actor Object Reference=@n3) -> (ReturnValue:Fort Minigame Object Reference)
   > Is Valid(InputObject=@n4)
   > SetItemCost!(
    execute
    self:Item Cost Widget Object Reference=Get ItemCostWidget() -> (ItemCostWidget:Item Cost Widget Object Reference)
    Actor:Actor Object Reference=Select(
     Option 0:Actor Object Reference
     Option 1:Actor Object Reference=@n3
     Index=IsInPlayMode!(self:Fort Minigame Object Reference=@n4) -> (ReturnValue:bool)
    ) -> (ReturnValue:Actor Object Reference)
    SpecifiedItemCostComponent:Item Cost Component Class Reference=/Game/Creative/Devices/Common/Components/FNEDeviceCostComponent.FNEDeviceCostComponent_C
   ) -> (then)
   n5 = FindParentWidgetOfType!(
    execute=@n2.else
    StartingWidget:Widget Object Reference=GetParent!(self:Widget Object Reference) -> (ReturnValue:Panel Widget Object Reference)
    Type:Widget Class Reference=/Game/Athena/HUD/AthenaInteractionIndicator.AthenaInteractionIndicator_C
   ) -> (then, ReturnValue:Athena Interaction Indicator Object Reference)
   > n6 = Set Indicator Widget(Indicator Widget:Athena Interaction Indicator Object Reference=@n5) -> (Output_Get:Athena Interaction Indicator Object Reference)
  ]}
 Update Details From Context {nodes [
   n1 = Update Details From Context!() -> (then, New Context:Fort Interact Context Info Object Reference) {is=title K2Node_FunctionEntry}
   > n2 = ShouldShowCreativeModeName!(
    execute
    self:Athena Interaction Indicator Object Reference=Get Indicator Widget() -> (Indicator Widget:Athena Interaction Indicator Object Reference)
   ) -> (then, Show:bool, CreativeNameText:Text)
   n3 = Get ContextText!(self:Fort Interact Context Info Object Reference=Get New Context() -> (New Context:Fort Interact Context Info Object Reference)) -> (ContextText:Text)
   > UpdateDescription!(
    execute
    self:Basic Interaction Widget Base Object Reference=Get BasicInteractionWidget() -> (BasicInteractionWidget:Basic Interaction Widget Object Reference)
    Description:Text=Select(
     Option 0:Text=@n3
     Option 1:Text=Format Text!(Format="LOCTEXT(\"{context}\\r\\n{name}\")", context:Text=@n3, name:Text=@n2.CreativeNameText) -> (Result)
     Index=@n2
    ) -> (ReturnValue:Text)
    DescriptionColor:Linear Color Structure=Get ContextTextColor!(self:Fort Interact Context Info Object Reference=@n1) -> (ContextTextColor:Linear Color Structure)
    SubDescription:Text
    SubDescriptionColor:Linear Color Structure
    InteractionCount:int=Get Interact Actor Count!(
     self:Athena Interaction Indicator Object Reference=Get Indicator Widget()
     CurrentInteraction:Fort Interact Context Info Object Reference=@n1
     MultiInteractComponent:Fort Actor Component Multi Interact Object Reference
     AppliesToInteraction:EInteractionBeingAttempted Enum=FirstInteraction
    ) -> (InteractionCount:int=0) {is=title K2Node_CallFunction}
   ) -> (then)
  ]}
]
variables [
 Context Info {type=FortInteractContextInfo, default=null}
 Indicator Widget {type=AthenaInteractionIndicator_C, default=null}
 ItemCostComponent {type=ItemCostComponent_C, default=null}
]
widgets {
 info {
  inheritedWidgetCount=0
  parentClass=/Script/FortniteGame.FortInteractContextInfoWidget
  rootWidgetClass=&VerticalBox
  widgetCount=3
 }
 widgets [
  VerticalBox_0 {widgetClassPath=&VerticalBox, widget=VerticalBox_0}
  BasicInteractionWidget {
   widgetClassPath=/Game/UI/InGame/HUD/BasicInteractionWidget
   bIsVariable=true
   parent=VerticalBox_0
   widget=BasicInteractionWidget
   slot {class=VerticalBoxSlot, properties {size {sizeRule=Fill, value=1}}}
  }
  ItemCostWidget {
   widgetClassPath=/Game/UI/Creative/ItemCostWidget
   bIsVariable=true
   parent=VerticalBox_0
   widget=ItemCostWidget
   slot {class=VerticalBoxSlot}
  }
 ]
}
